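# Translucent UI material that draws the pip decoration for the Rift Gauntlets interface from an MSDF texture.
# The texture converts to a signed distance field, which two anti-aliasing functions threshold at different distances into shape masks.
# The first mask blends in the Color parameter for the final color.
# ShadowOpacity blends between both masks, creating a shadow from the same field, and Opacity scales the resulting alpha.
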Material /RiftGauntletsGameplay/UI/Art/Materials/M_UI_RiftGauntlets_Pip_Decor
graph {
 properties {
  bAutomaticallySetUsageInEditor=false
  bUsedWithStaticMesh=false
  blendMode=BLEND_Translucent
  materialDomain=MD_UI
  usedShadingModels=Default Lit
 }
 nodes [
  n1 = MaterialFunctionCall!(MSDF RGB=TextureSample(UVs=TextureCoordinate()) {samplerType=SAMPLERTYPE_LinearColor, texture=/RiftGauntletsGameplay/UI/Art/Textures/T_UI_RiftGauntlets_PipDecor}.RGB) {materialFunction=/FortUILibrary/MaterialFunctions/Shapes/Modifiers/MF_UI_MSDFtoSDF}
  n2 = MaterialFunctionCall(DistanceField=@n1.SDF, Threshold=Constant() {r=0.2}) {materialFunction=&MF_UI_AntiAliasing}
  NamedRerouteDeclaration(None=LinearInterpolate(B=VectorParameter() {parameterName=Color, defaultValue=<rgba 1 1 1 0>}.RGB, Alpha=@n2.Outer)) {name=PipDecorFinalColor, nodeColor=<rgba 1 0.2344 0 1>}
  n3 = MaterialFunctionCall(DistanceField=@n1.SDF, Threshold=Constant() {r=0.5}) {materialFunction=&MF_UI_AntiAliasing}
  n4 = ScalarParameter() {defaultValue=0.2, parameterName=ShadowOpacity}
  NamedRerouteDeclaration(None=Multiply(A=LinearInterpolate(A=@n2.Outer, B=@n3.Outer, Alpha=@n4), B=ScalarParameter() {defaultValue=0.7, parameterName=Opacity})) {name=PipDecorOpacity, nodeColor=<rgba 0 0.0859 1 1>}
 ]
}
