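# Rift Gauntlets reticle widget that shows a row of recharge-stack pips framed by decorative end caps.
# It subclasses the shared recharge-stack pip collection base and, on Construct, passes its DynamicEntryBox to that base through SetDynamicEntryBox.
# The entry box spawns Rift Gauntlets pip widgets with negative spacing so they overlap, between two images sharing one decor material, the left one mirrored by a negative X scale.
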WidgetBlueprint /RiftGauntletsGameplay/UI/Widgets/WBP_UI_RiftGauntlets_Collection
graphs [EventGraph {nodes [
   Event Construct()
   > SetDynamicEntryBox!(
    execute
    self:Fort Weapon Reticle Recharge Stack Base Pip Collection Base Object Reference
    DynamicEntryBox:Dynamic Entry Box Object Reference=Get DynamicEntryBox_Pips() -> (DynamicEntryBox_Pips:Dynamic Entry Box Object Reference)
   ) -> (then)
  ]}]
variables []
widgets {
 info {
  inheritedWidgetCount=0
  parentClass=/Game/Athena/HUD/reticles/ReticleExtensionWidgets/ChargeStack/WBP_Reticle_RechargeStackBase_PipCollection.WBP_Reticle_RechargeStackBase_PipCollection_C
  rootWidgetClass=&SizeBox
  widgetCount=5
 }
 widgets [
  SizeBox_27 {widgetClassPath=&SizeBox, widget=SizeBox_27}
  HorizontalBox_0 {
   widgetClassPath=&HorizontalBox
   parent=SizeBox_27
   widget=HorizontalBox_0
   slot {class=SizeBoxSlot, properties {horizontalAlignment=HAlign_Center, verticalAlignment=VAlign_Center}}
  }
  Image_DetailL {
   widgetClassPath=&Image
   parent=HorizontalBox_0
   widget=Image_DetailL
   properties {
    visibility=SelfHitTestInvisible
    brush {
     bIsDynamicallyLoaded=false
     drawAs=Image
     imageType=FullColor
     mirroring=NoMirror
     resourceName=None
     resourceObject=/RiftGauntletsGameplay/UI/Art/Materials/M_UI_RiftGauntlets_Pip_Decor
     tiling=NoTile
     imageSize=<16 16>
     margin=<ltrb 0 0 0 0>
     outlineSettings {
      bUseBrushTransparency=false
      roundingType=HalfHeightRadius
      width=0
      color {colorUseRule=UseColor_Specified, specifiedColor=<rgba 0 0 0 0>}
      cornerRadii=<0 0 0 0>
     }
     tintColor {colorUseRule=UseColor_Specified, specifiedColor=<rgba 1 1 1 1>}
     uVRegion {bIsValid=false, max=<0 0>, min=<0 0>}
    }
    renderTransform {angle=0, scale=<-1 1>, shear=<0 0>, translation=<0 0>}
   }
   slot {class=HorizontalBoxSlot, properties {horizontalAlignment=HAlign_Center, verticalAlignment=VAlign_Center, padding=<ltrb 0 -2 -1 0>}}
  }
  DynamicEntryBox_Pips {
   widgetClassPath=&DynamicEntryBox
   bIsVariable=true
   parent=HorizontalBox_0
   widget=DynamicEntryBox_Pips
   properties {
    entryWidgetClass=&/WBP_UI_RiftGauntlets_Pip.WBP_UI_RiftGauntlets_Pip_C
    numDesignerPreviewEntries=4
    entrySpacing=<-8 0>
    radialBoxSettings {angleBetweenItems=21, bClockwiseOrder=false, bDistributeItemsEvenly=false, marginSize=0, sectorCentralAngle=360, startingAngle=260}
   }
   slot {class=HorizontalBoxSlot}
  }
  Image_DetailR {
   widgetClassPath=&Image
   parent=HorizontalBox_0
   widget=Image_DetailR
   properties {
    visibility=SelfHitTestInvisible
    brush {
     bIsDynamicallyLoaded=false
     drawAs=Image
     imageType=FullColor
     mirroring=NoMirror
     resourceName=None
     resourceObject=&M_UI_RiftGauntlets_Pip_Decor
     tiling=NoTile
     imageSize=<16 16>
     margin=<ltrb 0 0 0 0>
     outlineSettings {
      bUseBrushTransparency=false
      roundingType=HalfHeightRadius
      width=0
      color {colorUseRule=UseColor_Specified, specifiedColor=<rgba 0 0 0 0>}
      cornerRadii=<0 0 0 0>
     }
     tintColor {colorUseRule=UseColor_Specified, specifiedColor=<rgba 1 1 1 1>}
     uVRegion {bIsValid=false, max=<0 0>, min=<0 0>}
    }
   }
   slot {class=HorizontalBoxSlot, properties {horizontalAlignment=HAlign_Center, verticalAlignment=VAlign_Center, padding=<ltrb -1 -2 0 0>}}
  }
 ]
}
