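# Translucent UI material for Rift Gauntlets that draws ammo pips from MF_Pip.
# Scaled UI UVs form repeating cell UVs, though the cell count divides pixel width by itself, yielding one cell.
# Step masks against CurrentAmmo and MaxAmmo mark completed and available cells.
# MF_Pip instances supply complete, incomplete and parameter-driven Pip1 looks, with Lerp and Max nodes for compositing opacity and color.
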Material /RiftGauntletsGameplay/UI/Art/Materials/M_UI_RiftGauntlets_CombinedPips1
graph {
 properties {
  bAutomaticallySetUsageInEditor=false
  bUsedWithStaticMesh=false
  blendMode=BLEND_Translucent
  materialDomain=MD_UI
  usedShadingModels=Default Lit
 }
 nodes [
  n1 = MaterialFunctionCall!(
   SizeMultiplier=ScalarParameter() {defaultValue=0.35, parameterName=SizeMP}
   Progress=Constant()
   UV=NamedRerouteUsage()
   Gradient
   ProgressOpacity
   InPanner
  ) {materialFunction=/RiftGauntletsGameplay/UI/Art/Materials/MF/MF_Pip}
  LinearInterpolate(
   A=NamedRerouteUsage()
   B=NamedRerouteDeclaration(None=@n1.Opacity) {name=IncompleteOpacity, nodeColor=<rgba 1 0 0.7734 1>}
   Alpha=Subtract(A=NamedRerouteUsage(), B=NamedRerouteUsage())
  )
  MaterialFunctionCall(
   SizeMultiplier=ScalarParameter() {defaultValue=0.35, parameterName=Pip1Size}
   Progress=ScalarParameter() {defaultValue=1, parameterName=Pip1Progress}
   UV=NamedRerouteUsage()
   Gradient=ScalarParameter() {parameterName=Pip1Gradient}
   ProgressOpacity=ScalarParameter() {parameterName=Pip1Opacity}
  ) {materialFunction=&MF_Pip}
  Max(A=NamedRerouteUsage(), B=NamedRerouteDeclaration(None=@n1.Color) {name=Composite_IncompleteColors, nodeColor=<rgba 0.75 0.6098 0.375 0.75>})
  NamedRerouteDeclaration(None=MaterialFunctionCall(UV=MaterialFunctionCall() {materialFunction=&GetUserInterfaceUV}.Normalized UV) {materialFunction=&MF_UI_UVScaling}.Scaled To Fill) {name=SepUVs, nodeColor=<rgba 1 0 0.9375 1>}
  n2 = MaterialFunctionCall() {materialFunction=&GetUserInterfaceUV}
  n3 = MaterialFunctionCall(Float2=MaterialFunctionCall(UV=@n2.Normalized UV) {materialFunction=&MF_UI_UVScaling}.ScaleY) {materialFunction=&BreakOutFloat2Components}
  n4 = MaterialFunctionCall(Float2=@n2.Pixel Size) {materialFunction=&BreakOutFloat2Components}
  n5 = Floor(None=Divide(A=@n4.R, B=@n4.R) {desc=Aspect Ratio}) {desc=Total Cells}
  n6 = Multiply(A=@n5) {constB=-0.5, desc=OffsetFirstCell}
  n7 = Clamp(None=Add(A=Subtract(A=@n3.R, B=@n6)) {constB=-0.5, desc=OffsetCellFromSide}, Max=@n5) {desc=StretchEdge}
  NamedRerouteDeclaration(None=Step(Y=@n7, X=ScalarParameter() {parameterName=CurrentAmmo})) {name=Input_NestsCompleted, nodeColor=<rgba 0.75 0.6098 0.375 0.75>}
  n8 = MaterialFunctionCall(
   SizeMultiplier=ScalarParameter() {defaultValue=0.35, parameterName=SizeMP}
   Progress=ScalarParameter() {defaultValue=1, parameterName=PipProgress}
   UV=NamedRerouteUsage()
  ) {materialFunction=&MF_Pip}
  NamedRerouteDeclaration(None=@n8.Opacity) {name=Composite_SquareOpacity, nodeColor=<rgba 0.75 0.6098 0.375 0.75>}
  n9 = AppendVector(A=@n7, B=@n3.G)
  NamedRerouteDeclaration(None=MaterialFunctionCall(UVs=Frac(None=@n9), Texture Scale=Constant() {r=1.4}) {materialFunction=&ScaleUVsByCenter}.UVs) {name=RepeatUVs, nodeColor=<rgba 0.75 0.6098 0.375 0.75>}
  NamedRerouteDeclaration(None=@n8.Color) {name=Color_CompleteBox, nodeColor=<rgba 0.75 0.6098 0.375 0.75>}
  NamedRerouteDeclaration(None=Step(Y=@n7, X=ScalarParameter() {defaultValue=7, parameterName=MaxAmmo}) {constX=4}) {name=Input_NestAmount, nodeColor=<rgba 0.75 0.6098 0.375 0.75>}
 ]
}
