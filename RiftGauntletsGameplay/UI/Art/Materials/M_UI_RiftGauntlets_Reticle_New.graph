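# Translucent UI material that draws the Rift Gauntlets reticle, reacting to dash, jump, slam/punch, in-range and player-direction parameters.
# SDF functions build a center dashed ring and dot, outer bracket rings and a rotating dash-direction indicator, plus MSDF texture corner details.
# Counter-rotating radial segments create a distortion mask for the slam/punch ring warp.
# Shapes and shadows are composited using shared MPC white/charcoal, orange accents and opacity parameters.
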Material /RiftGauntletsGameplay/UI/Art/Materials/M_UI_RiftGauntlets_Reticle_New
graph {
 properties {
  bAutomaticallySetUsageInEditor=false
  bUsedWithStaticMesh=false
  blendMode=BLEND_Translucent
  materialDomain=MD_UI
  usedShadingModels=Default Lit
 }
 nodes [
  n1 = CollectionParameter() {collection=/BattleRoyaleGameUI/Shared/MPC_BRGameUI, parameterName=White}
  n2 = Constant4Vector() {constant=<rgba 0.7835 0.2384 0.003 1>}
  n3 = Add(A=ScalarParameter() {parameterName=ActiveDash}, B=ScalarParameter() {parameterName=ActiveJump})
  n4 = Add(
   A=CollectionParameter() {collection=&MPC_BRGameUI, parameterName=Charcoal}
   B=Multiply(A=NamedRerouteUsage(), B=LinearInterpolate(A=@n1, B=@n2.RGBA, Alpha=@n3))
  )
  n5 = Multiply(A=NamedRerouteUsage(), B=CollectionParameter() {collection=&MPC_BRGameUI, parameterName=White})
  n6 = NamedRerouteUsage()
  n7 = Constant4Vector() {constant=<rgba 0.7835 0.2384 0.003 1>}
  n8 = Add(
   A=Max(A=Max(A=@n4, B=@n5), B=Multiply(A=@n6, B=@n7.RGBA))
   B=Multiply(A=NamedRerouteUsage(), B=CollectionParameter() {collection=&MPC_BRGameUI, parameterName=White}) {constB=0}
  )
  n9 = Multiply(A=NamedRerouteUsage(), B=CollectionParameter() {collection=&MPC_BRGameUI, parameterName=Charcoal}) {constB=0}
  n10 = NamedRerouteUsage()
  n11 = CollectionParameter() {collection=&MPC_BRGameUI, parameterName=Charcoal}
  Add(
   A=Add(A=Add(A=@n8, B=@n9), B=Multiply(A=@n10, B=@n11) {constB=0})
   B=Multiply(A=NamedRerouteUsage(), B=CollectionParameter() {collection=&MPC_BRGameUI, parameterName=White}) {constB=0}
  )
  n12 = NamedRerouteUsage()
  n13 = LinearInterpolate(Alpha=ScalarParameter() {parameterName=In Range}) {constA=0.5, constAlpha=0, constB=0.9}
  n14 = Max(
   A=LinearInterpolate(A=Multiply(A=@n12, B=@n13) {constB=0.5}, B=NamedRerouteUsage(), Alpha=ScalarParameter() {parameterName=Slam/Punch})
   B=Multiply(A=NamedRerouteUsage())
  )
  n15 = Multiply(A=NamedRerouteUsage()) {constB=0.8}
  n16 = NamedRerouteUsage()
  n17 = Constant() {r=0.2}
  n18 = Add(A=Add(A=Add(A=@n14, B=@n15), B=Multiply(A=@n16, B=@n17)), B=Multiply(A=NamedRerouteUsage(), B=Constant() {r=0.25}))
  n19 = Multiply(A=NamedRerouteUsage(), B=ScalarParameter() {parameterName=Dash Direction Opacity})
  n20 = Multiply(A=NamedRerouteUsage(), B=Constant() {r=0.25})
  n21 = ScalarParameter() {parameterName=Dash Direction Opacity}
  Multiply(A=Add(A=Add(A=@n18, B=@n19), B=Multiply(A=@n20, B=@n21)), B=ScalarParameter() {defaultValue=1, parameterName=Opacity})
  n22 = NamedRerouteUsage()
  n23 = Time()
  n24 = Constant() {r=-0.1}
  n25 = Constant()
  n26 = MaterialFunctionCall!(
   UV=MaterialFunctionCall(UVs=@n22, "Rotation Angle (0-1)"=Multiply(A=@n23, B=@n24) {constB=0.5}) {materialFunction=&CustomRotator}.Rotated Values
   Segment Count=Constant() {r=20}
   Segment Spacing=@n25
   Fixed Spacing
  ) {materialFunction=/FortUILibrary/MaterialFunctions/Basics/MF_UI_RadialSegments}
  n27 = Constant() {r=0.1}
  n28 = MaterialFunctionCall(
   UV=MaterialFunctionCall(UVs=@n22, "Rotation Angle (0-1)"=Multiply(A=@n23, B=@n27) {constB=0.5}) {materialFunction=&CustomRotator}.Rotated Values
   Segment Count=Constant() {r=15}
   Segment Spacing=@n25
  ) {materialFunction=&MF_UI_RadialSegments}
  n29 = ScalarParameter() {parameterName=Slam/Punch}
  NamedRerouteDeclaration(None=Multiply(A=Multiply(A=@n26.Segmented Triangles Gradient, B=@n28.Segmented Triangles Gradient), B=LinearInterpolate(Alpha=@n29) {constB=1.2})) {name=Distortion, nodeColor=<rgba 0 0 0 1>}
  NamedRerouteDeclaration(None=MaterialFunctionCall() {materialFunction=&GetUserInterfaceUV}.Normalized UV) {name=UV, nodeColor=<rgba 0 0 0 1>}
  n30 = ScalarParameter() {defaultValue=1, parameterName=Player Direction}
  n31 = MaterialFunctionCall(
   UVs=NamedRerouteUsage()
   Rotation Center=Constant2Vector() {g=0.5, r=0.5}
   "Rotation Angle (0-1)"=Subtract(B=Divide(A=@n30) {constB=360, desc=Normalizing value received})
  ) {materialFunction=&CustomRotator}
  n32 = AppendVector(A=Constant() {r=0.11}, B=Constant() {r=0.14})
  n33 = Constant()
  n34 = ScalarParameter() {defaultValue=0.8, parameterName=DirectionalArrowDistance}
  n35 = MaterialFunctionCall!(
   UVs=Add(
    A=MaterialFunctionCall(UVs=@n31.Rotated Values, Texture Scale=@n32) {materialFunction=&ScaleUVsByCenter}.UVs
    B=AppendVector(A=@n33, B=@n34)
   ) {constB=0.1}
   Size
   Sides=Constant() {r=3}
   InnerAngle
   OutlineThickness
   OutlineAlignment
   GlowMin
   GlowMax
  ) {materialFunction=/FortUILibrary/MaterialFunctions/Shapes/MF_UI_RegularPolygon}
  NamedRerouteDeclaration(None=MaterialFunctionCall(DistanceField=@n35.SDF) {materialFunction=&MF_UI_AntiAliasing}.Inner) {name=Dash Direction Shape_1, nodeColor=<rgba 0 0 0 1>}
  n36 = ScalarParameter() {parameterName=ActiveDash}
  n37 = ScalarParameter() {parameterName=ActiveDash}
  n38 = LinearInterpolate(
   A=Add(A=LinearInterpolate(Alpha=@n37) {constA=0.022, constB=0.2}, B=NamedRerouteUsage()) {constB=0}
   B=Constant() {r=0.2}
   Alpha=ScalarParameter() {parameterName=ActiveJump}
  ) {constAlpha=1}
  n39 = MaterialFunctionCall() {materialFunction=&GetUserInterfaceUV}
  n40 = MaterialFunctionCall(
   UVs=@n39.Normalized UV
   Size=@n38
   OutlineThickness=LinearInterpolate(A=LinearInterpolate(Alpha=@n36) {constA=0.04, constB=0.042}, B=Constant() {r=0.04}, Alpha=ScalarParameter() {parameterName=ActiveJump}) {constAlpha=1}
   OutlineAlignment=Constant() {r=-0.5}
  ) {materialFunction=&MF_UI_Circle}
  n41 = MaterialFunctionCall() {materialFunction=&GetUserInterfaceUV}
  n42 = ScalarParameter() {parameterName=InnerCircleRingGapRotation}
  n43 = LinearInterpolate(B=Constant() {r=0.125}, Alpha=ScalarParameter() {parameterName=ActiveDash}) {constA=-0.1}
  n44 = NamedRerouteUsage()
  n45 = ScalarParameter() {parameterName=ActiveDash}
  n46 = MaterialFunctionCall(
   UVs=MaterialFunctionCall!(
    Reticle UVs=MaterialFunctionCall(UVs=@n41.Normalized UV, "Rotation Angle (0-1)"=@n42) {materialFunction=&CustomRotator}.Rotated Values
    Reticle Sides=Constant() {r=3}
    Reticle Scale
    Reticle Rotation
    Prong Scale
    Prong Rotation
    Spread Offset
    Positional Offset
    "Spread Input Range [Min,Max]"
    "Spread Remapped Range [Min,Max]"
   ) {materialFunction=/FortUILibrary/Systems/Reticles/MaterialFunctions/MF_UI_Reticle_UVs}.UVs
   Size=AppendVector(A=Add(A=@n43, B=@n44) {constB=0}, B=LinearInterpolate(Alpha=@n45))
  ) {materialFunction=&MF_UI_Box}
  NamedRerouteDeclaration(None=MaterialFunctionCall(DistanceField=MaterialFunctionCall!(DistanceFieldA=@n40.SDFOutline, DistanceFieldB=@n46.SDF, BlendAmount) {materialFunction=/FortUILibrary/MaterialFunctions/Shapes/Modifiers/MF_UI_SDF_Subtraction}.Exact) {materialFunction=&MF_UI_AntiAliasing}.Inner) {name=Center Dash Ring and Dot Shadow, nodeColor=<rgba 0 0 0 1>}
  n47 = MaterialFunctionCall(
   UVs=NamedRerouteUsage()
   Size=ScalarParameter() {defaultValue=0.3, parameterName=Outline Size}
   OutlineThickness=ScalarParameter() {defaultValue=0.02, parameterName=Outline Thickness}
  ) {materialFunction=&MF_UI_Circle}
  n48 = NamedRerouteUsage()
  n49 = Subtract(B=Divide(A=ScalarParameter() {defaultValue=1, parameterName=Player Direction}) {constB=360, desc=Normalizing value received})
  n50 = Add(
   A=MaterialFunctionCall(
    UVs=MaterialFunctionCall(UVs=@n48, "Rotation Angle (0-1)"=@n49) {materialFunction=&CustomRotator}.Rotated Values
    Texture Scale=Constant() {r=0.11}
   ) {materialFunction=&ScaleUVsByCenter}.UVs
   B=AppendVector(A=Constant(), B=ScalarParameter() {defaultValue=0.8, parameterName=DirectionalArrowDistance})
  ) {constB=0.1}
  n51 = Multiply(A=MaterialFunctionCall(UVs=@n50, Size=Constant2Vector() {g=2.5, r=1.12}) {materialFunction=&MF_UI_Box}.SDF) {constB=-1}
  NamedRerouteDeclaration(None=MaterialFunctionCall(
    DistanceField=MaterialFunctionCall(DistanceFieldA=@n47.SDFOutline, DistanceFieldB=@n51) {materialFunction=&MF_UI_SDF_Subtraction}.Exact
    Threshold=Constant() {r=0.005}
   ) {materialFunction=&MF_UI_AntiAliasing}.Inner) {name=Dash Direction Shadow, nodeColor=<rgba 0 0 0 1>}
  NamedRerouteDeclaration(None=MaterialFunctionCall(DistanceField=@n35.SDF, Threshold=Constant() {r=0.1}) {materialFunction=&MF_UI_AntiAliasing}.Inner) {name=Dash Direction Shadow_1, nodeColor=<rgba 0 0 0 1>}
  n52 = ScalarParameter() {desc=Variable to Position Reticle Prongs, group=Reticle Operators, parameterName=PositionalOffset, sliderMax=1, sliderMin=-1}
  n53 = VectorParameter() {group=Reticle Operators, parameterName=SpreadInputRange, sortPriority=33, defaultValue=<rgba 0 100 0 0>}
  n54 = VectorParameter() {group=Reticle Operators, parameterName=SpreadRemappedRange, sortPriority=33, defaultValue=<rgba 0 1 0 0>}
  NamedRerouteDeclaration(None=MaterialFunctionCall!(
    Positional Spread Offset=Multiply(A=@n52, B=@n53.G)
    Accuracy Spread Offset=ScalarParameter() {desc=Variable for Accuracy Spread Data, group=Reticle Operators, parameterName=SpreadOffset}
    Min Value
    "Input Range [Min,Max]"=@n53.RGB
    "Remapped Range [Min,Max]"=@n54.RGB
   ) {materialFunction=/FortUILibrary/Systems/Reticles/MaterialFunctions/MF_UI_Reticle_Remap_Spread}.Result) {name=Spread/Bloom, nodeColor=<rgba 0.025 0.0203 0.0125 0.75>}
  n55 = ScalarParameter() {parameterName=ActiveDash}
  n56 = MaterialFunctionCall(
   UVs=@n39.Normalized UV
   Size=@n38
   OutlineThickness=LinearInterpolate(A=LinearInterpolate(Alpha=@n55) {constA=0.1, constB=0.017}, B=Constant() {r=0.02}, Alpha=ScalarParameter() {parameterName=ActiveJump}) {constAlpha=1}
   OutlineAlignment=Constant() {r=-1}
  ) {materialFunction=&MF_UI_Circle}
  NamedRerouteDeclaration(None=MaterialFunctionCall(DistanceField=MaterialFunctionCall(DistanceFieldA=@n56.SDFOutline, DistanceFieldB=@n46.SDF) {materialFunction=&MF_UI_SDF_Subtraction}.Exact) {materialFunction=&MF_UI_AntiAliasing}.Inner) {name=Center Dashed Ring And Dot, nodeColor=<rgba 0 0 0 1>}
  n57 = MaterialFunctionCall(UVs=Add(A=TextureCoordinate(), B=Constant2Vector()), Size=AppendVector(A=Constant() {r=1}, B=Constant() {r=0.38})) {materialFunction=&MF_UI_Box}
  n58 = ScalarParameter() {parameterName=In Range}
  n59 = Step(X=ScalarParameter() {parameterName=Slam/Punch}) {constX=0, constY=0.9}
  n60 = MaterialFunctionCall(
   UVs=MaterialFunctionCall(UVs=NamedRerouteUsage()) {materialFunction=&CustomRotator}.Rotated Values
   Size=LinearInterpolate(
    A=ScalarParameter() {defaultValue=0.68, parameterName=Brackets Distance}
    B=Constant() {r=0.58}
    Alpha=LinearInterpolate(A=@n58, Alpha=@n59) {constA=1, constB=0}
   )
   OutlineThickness=LinearInterpolate(A=Constant() {r=0.015}, B=Constant() {r=0.02}, Alpha=ScalarParameter() {parameterName=In Range})
  ) {materialFunction=&MF_UI_Circle}
  n61 = Constant() {r=0.005}
  NamedRerouteDeclaration(None=Subtract(
    A=MaterialFunctionCall(DistanceField=@n60.SDFOutline, Threshold=@n61) {materialFunction=&MF_UI_AntiAliasing}.Inner
    B=OneMinus(None=@n57.Fill)
   )) {name=Ring Simple, nodeColor=<rgba 0 0 0 1>}
  n62 = MaterialFunctionCall(UVs=Add(A=TextureCoordinate(), B=Constant2Vector()), Size=Constant2Vector() {g=0.38, r=1}) {materialFunction=&MF_UI_Box}
  n63 = Time()
  n64 = Add(
   A=MaterialFunctionCall(UVs=NamedRerouteUsage(), "Rotation Angle (0-1)"=Multiply(A=@n63) {constB=0}) {materialFunction=&CustomRotator}.Rotated Values
   B=LinearInterpolate(Alpha=ScalarParameter() {parameterName=Slam/Punch}) {constB=-0.005}
  ) {constB=0.01}
  n65 = OneMinus(None=NamedRerouteUsage())
  n66 = LinearInterpolate(
   B=LinearInterpolate(Alpha=ScalarParameter() {parameterName=Distortion}) {constA=0.02, constB=0.04}
   Alpha=ScalarParameter() {parameterName=Slam/Punch}
  ) {constB=0.02}
  n67 = Constant() {r=0.68}
  n68 = ScalarParameter() {parameterName=Slam/Punch}
  n69 = MaterialFunctionCall(DistanceField=MaterialFunctionCall(
    UVs=LinearInterpolate(A=@n64, B=@n65, Alpha=@n66) {constAlpha=0.05}
    Size=LinearInterpolate(A=@n67, Alpha=@n68) {constB=0.57}
    OutlineThickness=Constant() {r=0.023}
   ) {materialFunction=&MF_UI_Circle}.SDFOutline) {materialFunction=&MF_UI_AntiAliasing}
  n70 = LinearInterpolate(Alpha=ScalarParameter() {parameterName=Slam/Punch})
  NamedRerouteDeclaration(None=Subtract(A=Multiply(A=@n69.Inner, B=@n70) {constB=0.1}, B=OneMinus(None=@n62.Fill))) {name=Ring With Distortion, nodeColor=<rgba 0 0 0 1>}
  n71 = MaterialFunctionCall!(UV=TextureCoordinate(), Swizzle UV) {materialFunction=/FortUILibrary/MaterialFunctions/Basics/MF_UI_Symmetry_2x2_New}
  n72 = ScalarParameter() {parameterName=Details In Range}
  n73 = LinearInterpolate(
   A=Constant() {r=-0.21}
   B=Constant() {r=-0.32}
   Alpha=LinearInterpolate(A=ScalarParameter() {parameterName=Slam/Punch}, B=@n72, Alpha=@n72) {constAlpha=0}
  )
  n74 = Add(A=Constant() {r=0.3}, B=Constant() {r=-0.32})
  n75 = ScalarParameter() {parameterName=Details In Range Intro}
  n76 = AppendVector(
   A=Add(A=LinearInterpolate(A=@n73, B=@n74, Alpha=@n75), B=Constant() {r=0.32})
   B=LinearInterpolate(A=Constant() {r=-0.1}, B=Constant() {r=-0.1}, Alpha=ScalarParameter() {parameterName=In Range})
  )
  n77 = TextureSample(UVs=MaterialFunctionCall(UVs=Add(A=@n71.Symmetry, B=@n76), Texture Scale=Constant() {r=0.23}) {materialFunction=&ScaleUVsByCenter}.UVs) {
   samplerSource=SSM_Clamp_WorldGroupSettings
   samplerType=SAMPLERTYPE_LinearColor
   texture=/RiftGauntletsGameplay/UI/Art/Textures/T_UI_RiftGauntlets_PipDecor_Corners
  }
  n78 = MaterialFunctionCall!(MSDF RGB=@n77.RGB) {materialFunction=/FortUILibrary/MaterialFunctions/Shapes/Modifiers/MF_UI_MSDFtoSDF}
  NamedRerouteDeclaration(None=MaterialFunctionCall(DistanceField=@n78.SDF, Threshold=Constant() {r=0.8}) {materialFunction=&MF_UI_AntiAliasing}.Outer) {name=Detail, nodeColor=<rgba 0 0 0 1>}
  NamedRerouteDeclaration(None=MaterialFunctionCall(DistanceField=@n78.SDF, Threshold=Constant() {r=0.5}) {materialFunction=&MF_UI_AntiAliasing}.Outer) {name=DetailShadow, nodeColor=<rgba 0 0 0 1>}
  n79 = Constant() {r=0.02}
  n80 = MaterialFunctionCall(UVs=Add(A=TextureCoordinate(), B=Constant2Vector()), Size=AppendVector(A=Constant() {r=1}, B=Constant() {r=0.4})) {materialFunction=&MF_UI_Box}
  NamedRerouteDeclaration(None=Subtract(
    A=MaterialFunctionCall(DistanceField=@n60.SDFOutline, Threshold=@n79) {materialFunction=&MF_UI_AntiAliasing}.Inner
    B=OneMinus(None=@n80.Fill)
   )) {name=Ring Simple Shadow, nodeColor=<rgba 0 0 0 1>}
  n81 = Multiply(A=MaterialFunctionCall(UVs=@n50, Size=Constant2Vector() {g=2.5, r=1.1}) {materialFunction=&MF_UI_Box}.SDF) {constB=-1}
  NamedRerouteDeclaration(None=MaterialFunctionCall(DistanceField=MaterialFunctionCall(DistanceFieldA=@n47.SDFOutline, DistanceFieldB=@n81) {materialFunction=&MF_UI_SDF_Subtraction}.Exact) {materialFunction=&MF_UI_AntiAliasing}.Inner) {name=Dash Direction Shape, nodeColor=<rgba 0 0 0 1>}
 ]
}
