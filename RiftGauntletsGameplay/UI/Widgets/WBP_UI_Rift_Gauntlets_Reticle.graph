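# Rift Gauntlets reticle widget, extending the recharge-stack base with an animated material reticle and an ammo/cooldown sub-widget.
# On view-target change, it caches pawn and weapon, binds targeting feedback for in-range animation, and starts async gameplay tag/event listeners.
# These drive jump, dash and damage animations, plus a hotfix-gated timer feeding input direction to the dash arrow.
# Ammo changes update the sub-widget, cooldown polling and recharge sound.
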WidgetBlueprint /RiftGauntletsGameplay/UI/Widgets/WBP_UI_Rift_Gauntlets_Reticle
animations [
 AnimIntro {
  binding_count=2
  length_seconds=0.4167
  object=&self:AnimIntro
  sequence {
   bindings [
    WBP_UI_RiftGauntlets_MaterialOnly {
     id=#1
     possessed_class=WBP_UI_RiftGauntlets_MaterialOnly_C
     tracks [
      {
       type=MovieSceneFloatTrack
       display_name=Render Opacity
       sections [{
         type=MovieSceneFloatSection
         end_frame=24
         start_frame=-1
         channels [None {type=MovieSceneScriptingFloatChannel, keys [[-1,0],[0,0],[5,1],[24,1]]}]
        }]
      }
      {
       type=MovieScene2DTransformTrack
       sections [{
         type=MovieScene2DTransformSection
         end_frame=13
         start_frame=0
         channels [
          Translation.X {type=MovieSceneScriptingFloatChannel, const [0,1]}
          Translation.Y {type=MovieSceneScriptingFloatChannel, keys [[0,10],[13,0]]}
         ]
        }]
      }
     ]
    }
    Image_Reticle {
     id=#2
     possessed_class=Image
     tracks [{
       type=MovieSceneWidgetMaterialTrack
       sections [{
         type=MovieSceneParameterSection
         end_frame=24
         start_frame=0
         channels [
          SpreadOffset {type=MovieSceneScriptingFloatChannel, keys [[3,4],[14.199,1]]}
          Brackets Distance {type=MovieSceneScriptingFloatChannel, keys [[9,0.85],[21.8,0.68]]}
          Details In Range Intro {type=MovieSceneScriptingFloatChannel, keys [[16,1],[21,0]]}
          Opacity {type=MovieSceneScriptingFloatChannel, keys [[0,0],[7,1]]}
         ]
        }]
      }]
    }
   ]
   display_rate {numerator=60}
  }
 }
 AnimInRange {
  binding_count=1
  length_seconds=0.35
  object=&self:AnimInRange
  sequence {bindings [Image_Reticle {
     id=#3
     possessed_class=Image
     tracks [{
       type=MovieSceneWidgetMaterialTrack
       sections [{
         type=MovieSceneParameterSection
         end_frame=7
         start_frame=0
         channels [
          In Range {type=MovieSceneScriptingFloatChannel, keys [[0,0],[7,1]]}
          Details In Range {type=MovieSceneScriptingFloatChannel, keys [[4,0],[7,1]]}
         ]
        }]
      }]
    }]}
 }
 AnimDash {
  binding_count=1
  length_seconds=0.7
  object=&self:AnimDash
  sequence {bindings [Image_Reticle {
     id=#4
     possessed_class=Image
     tracks [{
       type=MovieSceneWidgetMaterialTrack
       sections [{
         type=MovieSceneParameterSection
         end_frame=14
         start_frame=0
         channels [
          ActiveDash {type=MovieSceneScriptingFloatChannel, const [0,4]}
          InnerCircleRingGapRotation {type=MovieSceneScriptingFloatChannel, const [0,2]}
          ActiveJump {type=MovieSceneScriptingFloatChannel, keys [[0,0],[3,1],[11,1],[14,0]]}
         ]
        }]
      }]
    }]}
 }
 AnimJump {
  binding_count=1
  length_seconds=0.65
  object=&self:AnimJump
  sequence {bindings [Image_Reticle {
     id=#4
     possessed_class=Image
     tracks [{
       type=MovieSceneWidgetMaterialTrack
       sections [
        {type=MovieSceneParameterSection, end_frame=6, start_frame=0}
        {
         type=MovieSceneParameterSection
         end_frame=13
         start_frame=0
         channels [ActiveJump {type=MovieSceneScriptingFloatChannel, keys [[0,0],[4,1],[11,1],[13,0]]}]
        }
       ]
      }]
    }]}
 }
 AnimDashPunch {
  binding_count=1
  length_seconds=1.05
  object=&self:AnimDashPunch
  sequence {bindings [Image_Reticle {
     id=#5
     possessed_class=Image
     tracks [{
       type=MovieSceneWidgetMaterialTrack
       sections [
        {type=MovieSceneParameterSection, end_frame=13, start_frame=-2}
        {
         type=MovieSceneParameterSection
         end_frame=19
         start_frame=-2
         channels [
          Slam/Punch {type=MovieSceneScriptingFloatChannel, keys [[-2,0],[1,1],[16,1],[19,0]]}
          Distortion {type=MovieSceneScriptingFloatChannel, const [0,3]}
         ]
        }
       ]
      }]
    }]}
 }
 AnimDashArrow {
  binding_count=1
  length_seconds=0.95
  object=&self:AnimDashArrow
  sequence {bindings [Image_Reticle {
     id=#6
     possessed_class=Image
     tracks [{
       type=MovieSceneWidgetMaterialTrack
       sections [{
         type=MovieSceneParameterSection
         end_frame=16
         start_frame=-3
         channels [
          DirectionalArrowDistance {type=MovieSceneScriptingFloatChannel, const [0.9814,4]}
          Dash Direction Opacity {type=MovieSceneScriptingFloatChannel, thinned_from=5, keys [[-3,0.5],[10,0.5],[15,0],[16,0]]}
          Outline Size {type=MovieSceneScriptingFloatChannel, thinned_from=5, keys [[-3,0.3],[0,0.3],[6,0.4],[16,0.4]]}
          Outline Thickness {type=MovieSceneScriptingFloatChannel, const [0.02,3]}
         ]
        }]
      }]
    }]}
 }
]
graphs [
 ClearAndCancelActions {nodes [
   ClearAndCancelActions() {is=title K2Node_FunctionEntry}
   > Cancel(self:Cancellable Async Action Object Reference=Get DashAbilityTriggeredAsyncAction() -> (DashAbilityTriggeredAsyncAction:Ability Async Wait Gameplay Tag Added Object Reference))
   > Cancel(self=Get DashMovementStartAsyncAction() -> (DashMovementStartAsyncAction:Ability Async Wait Gameplay Event Object Reference))
   > Cancel(self=Get PerformedDamageAsyncAction() -> (PerformedDamageAsyncAction:Ability Async Wait Gameplay Tag Added Object Reference))
   > Cancel(self=Get JumpAsyncAction() -> (JumpAsyncAction:Ability Async Wait Gameplay Event Object Reference))
   > Cancel(self=Get DashMovementEndedAsyncAction() -> (DashMovementEndedAsyncAction:Ability Async Wait Gameplay Event Object Reference))
   > Clear and Invalidate Timer by Handle(Handle=Get DashArrowEventTimerHandle() -> (DashArrowEventTimerHandle:Timer Handle Structure))
  ]}
 EventGraph {nodes [
   StartListeningToDashMovementStart() {is=title K2Node_CustomEvent}
   > n1 = WaitGameplayEventToActor!(
    execute:Exec
    TargetActor:Actor Object Reference=Get ViewTargetPlayerPawn() -> (ViewTargetPlayerPawn:Fort Player Pawn Object Reference)
    EventTag:Gameplay Tag Structure=Get DashMovementStartedGameplayTag() -> (DashMovementStartedGameplayTag:Gameplay Tag Structure)
    OnlyTriggerOnce:bool=false
    OnlyMatchExact:bool=true
   ) -> (then:Exec, AsyncTaskProxy:Ability Async Wait Gameplay Event Object Reference, EventReceived:Exec, Payload:Gameplay Event Data Structure) {is=title K2Node_AsyncAction}
   > Set DashMovementStartAsyncAction(DashMovementStartAsyncAction:Ability Async Wait Gameplay Event Object Reference=@n1) -> (Output_Get:Ability Async Wait Gameplay Event Object Reference)
   PlayAnimation(execute=@n1.EventReceived, InAnimation=Get AnimDash() -> (AnimDash:Widget Animation Object Reference))
   > Clear and Invalidate Timer by Handle(Handle=Get DashArrowEventTimerHandle())
   > Branch(Condition=GetValueAsBool!(
     Input:"Scalable Float Structure (by ref)"=Get Hotfix_EnableDashDirectionArrowHUD() -> (Hotfix_EnableDashDirectionArrowHUD:Scalable Float Structure)
     Level:float=0
    ) -> (ReturnValue:bool))
   > PlayAnimation(InAnimation=Get AnimDashArrow() -> (AnimDashArrow:Widget Animation Object Reference))
   StartListeningToJumpAbility() {is=title K2Node_CustomEvent}
   > n2 = WaitGameplayEventToActor(
    TargetActor=Get ViewTargetPlayerPawn()
    EventTag=Get JumpAbilityStartedTag() -> (JumpAbilityStartedTag:Gameplay Tag Structure)
    OnlyTriggerOnce=false
    OnlyMatchExact=true
   )
   > Set JumpAsyncAction(JumpAsyncAction:Ability Async Wait Gameplay Event Object Reference=@n2) -> (Output_Get:Ability Async Wait Gameplay Event Object Reference)
   PlayAnimation(execute=@n2.EventReceived, InAnimation=Get AnimJump() -> (AnimJump:Widget Animation Object Reference))
   StartListeningToDashAbility() {is=title K2Node_CustomEvent}
   > n3 = WaitGameplayTagAddToActor!(
    execute:Exec
    TargetActor:Actor Object Reference=Get ViewTargetPlayerPawn()
    Tag:Gameplay Tag Structure=Get DashAbilityStartedTag() -> (DashAbilityStartedTag:Gameplay Tag Structure)
    OnlyTriggerOnce:bool=false
   ) -> (then:Exec, AsyncTaskProxy:Ability Async Wait Gameplay Tag Added Object Reference, Added:Exec) {is=title K2Node_AsyncAction}
   > Set DashAbilityTriggeredAsyncAction(DashAbilityTriggeredAsyncAction:Ability Async Wait Gameplay Tag Added Object Reference=@n3) -> (Output_Get:Ability Async Wait Gameplay Tag Added Object Reference)
   OnDashAbilityTriggered(execute=@n3.Added)
   n4 = HandleTargetedActorTypeChanged!() -> (OutputDelegate, then, ActorType:AbilityTargetingTargetActorTypeEnum Enum=NewEnumerator0) {is=title K2Node_CustomEvent}
   > PlayAnimation(
    InAnimation=Get AnimInRange() -> (AnimInRange:Widget Animation Object Reference)
    PlayMode=Select!(
     NewEnumerator0:EUMGSequencePlayMode Enum=Reverse
     NewEnumerator1:EUMGSequencePlayMode Enum=Reverse
     NewEnumerator2:EUMGSequencePlayMode Enum
     Index:AbilityTargetingTargetActorTypeEnum Enum=@n4.ActorType
    ) -> (ReturnValue:EUMGSequencePlayMode Enum)
   )
   n5 = OnViewTargetChanged!() -> (
    OutputDelegate
    then
    PlayerController:Fort Player Controller Object Reference
    OldViewTarget:Actor Object Reference
    NewViewTarget:Actor Object Reference
   ) {is=title K2Node_CustomEvent}
   > n6 = Sequence!(execute) -> (then_0, then_1, then_2:Exec, then_3:Exec)
   n7 = Cast To FortPlayerPawn(execute=@n6.then_1, Object=@n5.NewViewTarget)
   > n8 = Set ViewTargetPlayerPawn(ViewTargetPlayerPawn:Fort Player Pawn Object Reference=@n7) -> (Output_Get:Fort Player Pawn Object Reference)
   > n9 = Cast To B_Weapon_RiftGauntlets!(execute, Object=GetCurrentWeapon!(self:Fort Pawn Object Reference=@n8) -> (ReturnValue:Fort Weapon Object Reference)) -> (then, CastFailed, AsB Weapon Rift Gauntlets:B Weapon Rift Gauntlets Object Reference) {is=title K2Node_DynamicCast}
   > n10 = Set RiftGauntletWeaponComponent(RiftGauntletWeaponComponent:B Weapon Rift Gauntlets Object Reference=@n9) -> (Output_Get:B Weapon Rift Gauntlets Object Reference)
   n11 = GetComponentByClass!(
    self:Actor Object Reference=@n10
    ComponentClass:Actor Component Class Reference=/RiftGauntletsGameplay/Gameplay/ReticleFeedback/B_RiftGauntletsAbilityTargetingFeedbackComponent.B_RiftGauntletsAbilityTargetingFeedbackComponent_C
   ) -> (ReturnValue:B Rift Gauntlets Ability Targeting Feedback Component Object Reference)
   > Is Valid(InputObject=@n11)
   > n12 = Set RiftGauntletTargetingFeedbackComponent(RiftGauntletTargetingFeedbackComponent:B Rift Gauntlets Ability Targeting Feedback Component Object Reference=@n11) -> (Output_Get:B Rift Gauntlets Ability Targeting Feedback Component Object Reference)
   > Bind Event to On Targeted Actor Type Changed(self:B Ability Targeting Feedback Component Parent Object Reference=@n12, Delegate=Create Event()) {is=title K2Node_AddDelegate}
   > HandleTargetedActorTypeChanged!(
    execute
    self
    ActorType:AbilityTargetingTargetActorTypeEnum Enum=GetLatestTargetingActorType!(self:B Ability Targeting Feedback Component Parent Object Reference=@n12) -> (Type:AbilityTargetingTargetActorTypeEnum Enum=NewEnumerator0)
   ) -> (then) {is=title K2Node_CallFunction}
   n13 = Get RiftGauntletWeaponComponent() -> (RiftGauntletWeaponComponent:B Weapon Rift Gauntlets Object Reference)
   > SetMaxAmmo!(
    execute
    self:WBP UI Rift Gauntlets Material Only Object Reference=Get WBP_UI_RiftGauntlets_MaterialOnly() -> (WBP_UI_RiftGauntlets_MaterialOnly:WBP UI Rift Gauntlets Material Only Object Reference)
    CurrentCompletedAmmo:int=GetBulletsPerClip!(self:Fort Weapon Object Reference=@n13) -> (ReturnValue:int=0)
   ) -> (then)
   > OnLocalAmmoChanged!(
    execute
    self:WBP Reticle Recharge Stack Base Object Reference
    LocalCount:int=GetLocalMagazineAmmoCount!(self:Fort Weapon Object Reference=@n13) -> (ReturnValue:int=0)
    LocalRemaining:int=GetLocalRemainingAmmo!(self:Fort Weapon Object Reference=@n13) -> (ReturnValue:int=0)
   ) -> (then)
   StartListeningToJumpAbility(execute=@n6.then_2) {is=title K2Node_CallFunction}
   > StartListeningToDashAbility() {is=title K2Node_CallFunction}
   > StartListeningToDashMovementStart() {is=title K2Node_CallFunction}
   > StartListeningToDashEndTrigger()
   > StartListeningToPerformedDamage()
   ResetMaterialProperties(execute=@n6.then_3)
   > PlayAnimation(InAnimation=Get AnimIntro() -> (AnimIntro:Widget Animation Object Reference))
   Is Valid(exec=@n6.then_0, InputObject=Get ViewTargetPlayerPawn())
   > Unbind Event from On Targeted Actor Type Changed(
    self:B Ability Targeting Feedback Component Parent Object Reference=Get RiftGauntletTargetingFeedbackComponent() -> (RiftGauntletTargetingFeedbackComponent:B Rift Gauntlets Ability Targeting Feedback Component Object Reference)
    Delegate=Create Event()
   ) {is=title K2Node_RemoveDelegate}
   > Set ViewTargetPlayerPawn()
   StartListeningToPerformedDamage() {is=title K2Node_CustomEvent}
   > n14 = WaitGameplayTagAddToActor(
    TargetActor=Get ViewTargetPlayerPawn()
    Tag=Get PerformedDamageGameplayTag() -> (PerformedDamageGameplayTag:Gameplay Tag Structure)
    OnlyTriggerOnce=false
   )
   > Set PerformedDamageAsyncAction(PerformedDamageAsyncAction:Ability Async Wait Gameplay Tag Added Object Reference=@n14) -> (Output_Get:Ability Async Wait Gameplay Tag Added Object Reference)
   n15 = Branch(execute=@n14.Added, Condition=IsAnimationPlaying(InAnimation=Get AnimDashPunch() -> (AnimDashPunch:Widget Animation Object Reference)))
   PlayAnimation(execute=@n15.else, InAnimation=Get AnimDashPunch())
   OnDashAbilityTriggered() {is=title K2Node_CustomEvent}
   > n16 = Branch(Condition=GetValueAsBool(Input=Get Hotfix_EnableDashDirectionArrowHUD(), Level=0))
   n17 = ShowDashDirectionalArrow() {is=title K2Node_CustomEvent}
   n18 = Set Timer by Event(execute=@n16.then, Delegate=@n17, Time=0.05, bLooping=true, bMaxOncePerFrame=true)
   > Set DashArrowEventTimerHandle(DashArrowEventTimerHandle:Timer Handle Structure=@n18) -> (Output_Get:Timer Handle Structure)
   > ShowDashDirectionalArrow() {is=title K2Node_CallFunction}
   > SetScalarParameterValue(
    self=Get ReticleDynamicMaterial() -> (ReticleDynamicMaterial:Material Instance Dynamic Object Reference)
    ParameterName=Dash Direction Opacity
    Value=0.5
   )
   > SetScalarParameterValue(self=Get ReticleDynamicMaterial(), ParameterName=Outline Size, Value=0.3)
   StartListeningToDashEndTrigger() {is=title K2Node_CustomEvent}
   > n19 = WaitGameplayEventToActor(
    TargetActor=Get ViewTargetPlayerPawn()
    EventTag=Get DashMovementEndedGameplayTag() -> (DashMovementEndedGameplayTag:Gameplay Tag Structure)
    OnlyTriggerOnce=false
    OnlyMatchExact=true
   )
   > Set DashMovementEndedAsyncAction(DashMovementEndedAsyncAction:Ability Async Wait Gameplay Event Object Reference=@n19) -> (Output_Get:Ability Async Wait Gameplay Event Object Reference)
   Clear and Invalidate Timer by Handle(execute=@n19.EventReceived, Handle=Get DashArrowEventTimerHandle())
   > SetScalarParameterValue(self=Get ReticleDynamicMaterial(), ParameterName=Dash Direction Opacity, Value=0)
   Event Destruct()
   > n20 = Sequence!(execute) -> (then_0, then_1)
   > n21 = Cast To FortPlayerController(Object=GetOwningPlayer())
   > Unbind Event from On View Target Changed(self:Fort Player Controller Object Reference=@n21, Delegate=Create Event()) {is=title K2Node_RemoveDelegate}
   > n22 = Get!(execute:Exec) -> (then:Exec, else:Exec, RiftGauntletTargetingFeedbackComponent:B Rift Gauntlets Ability Targeting Feedback Component Object Reference)
   > Unbind Event from On Targeted Actor Type Changed(self=@n22, Delegate=Create Event())
   > ClearAndCancelActions() {is=title K2Node_CallFunction}
   > HandleCooldownTimer!(execute, self:Fort Weapon Reticle Extension Widget Base Object Reference, bActivateTimer:bool=false, TimerInterval:float=0) -> (then)
   Parent: Destruct(execute=@n20.then_1)
   n23 = Event OnLocalAmmoChanged!() -> (OutputDelegate, then, LocalCount:int=0, LocalRemaining:int=0) {is=title K2Node_Event}
   > Parent: OnLocalAmmoChanged!(execute, LocalCount:int=@n23.LocalCount, LocalRemaining:int=@n23.LocalRemaining) -> (then) {is=title K2Node_CallParentFunction}
   > UpdateCooldownQueryTimer(self:Fort Weapon Reticle Extension Widget Base Object Reference)
   > OnAmmoChanged!(
    execute
    self:WBP UI Rift Gauntlets Material Only Object Reference=Get WBP_UI_RiftGauntlets_MaterialOnly()
    MaxAmmo:int=@n23.LocalCount
    CurrentAmmo:int=@n23.LocalCount
   ) -> (then)
   > n24 = Branch(Condition="integer > integer"(A=@n23.LocalCount, B=Get Audio_AmmoCount() -> (Audio_AmmoCount:int=0)))
   > n25 = PlaySound2D(Sound=/RiftGauntletsGameplay/Assets/Audio/MSS_RiftGauntlets_Cooldown_Complete)
   Set Audio_AmmoCount(execute=[@n25.then, @n24.else], Audio_AmmoCount:int=@n23.LocalCount) -> (Output_Get:int=0)
   Event OnInitialized()
   > n26 = GetDynamicMaterial(self=Get Image_Reticle() -> (Image_Reticle:Image Object Reference))
   > Set ReticleDynamicMaterial(ReticleDynamicMaterial:Material Instance Dynamic Object Reference=@n26) -> (Output_Get:Material Instance Dynamic Object Reference)
   Event UpdateRemainingCooldown() {is=title K2Node_Event}
   > Parent: UpdateRemainingCooldown() {is=title K2Node_CallParentFunction}
   > n27 = GetRemainingCooldownTimer!(execute, self) -> (then, RemainingTime:double, TotalDuration:double)
   > CooldownUpdated!(
    execute
    self:WBP UI Rift Gauntlets Material Only Object Reference=Get WBP_UI_RiftGauntlets_MaterialOnly()
    CurrentTime:double=@n27
    TotalTime:double=@n27.TotalDuration
   ) -> (then)
   Event UpdateCooldownQueryTimer() {is=title K2Node_Event}
   > Parent: UpdateCooldownQueryTimer() {is=title K2Node_CallParentFunction}
   > n28 = Get!(execute) -> (then, else, RiftGauntletWeaponComponent:B Weapon Rift Gauntlets Object Reference)
   > n29 = Branch(Condition="integer >= integer"(A=GetLocalMagazineAmmoCount(self=@n28) -> (ReturnValue=0), B=GetBulletsPerClip(self=@n28) -> (ReturnValue=0)))
   > HandleCooldownTimer(bActivateTimer=false, TimerInterval=0)
   HandleCooldownTimer(execute=@n29.else, bActivateTimer=true, TimerInterval=0.1)
   Event Construct()
   > Parent: Construct()
   > n30 = Cast To FortPlayerController(Object=GetOwningPlayer())
   > Bind Event to On View Target Changed(self:Fort Player Controller Object Reference=@n30, Delegate=Create Event()) {is=title K2Node_AddDelegate}
   > OnViewTargetChanged!(
    execute
    self
    PlayerController:Fort Player Controller Object Reference
    OldViewTarget:Actor Object Reference
    NewViewTarget:Actor Object Reference=GetViewTarget!(self:Controller Object Reference=@n30) -> (ReturnValue:Actor Object Reference)
   ) -> (then) {is=title K2Node_CallFunction}
   n31 = Get ViewTargetPlayerPawn()
   n32 = Get CharacterMovement!(self:Character Object Reference=@n31) -> (CharacterMovement:Character Movement Component Object Reference)
   SetScalarParameterValue(
    execute=@n17.then
    self=Get ReticleDynamicMaterial()
    ParameterName=Player Direction
    Value=CalculateDirection!(
     Velocity:"Vector (by ref)"=GetLastInputVector!(self:Pawn Movement Component Object Reference=@n32) -> (ReturnValue:Vector)
     BaseRotation:"Rotator (by ref)"=Get Actor Rotation!(self:Actor Object Reference=@n31) -> (ReturnValue:Rotator) {is=title K2Node_CallFunction}
    ) -> (ReturnValue:float)
   )
  ]}
 GetRemainingCooldownTimer {nodes [
   GetRemainingCooldownTimer() {is=title K2Node_FunctionEntry}
   > n1 = Get()
   n2 = GetInventoryItem!(self:Fort Weapon Object Reference=@n1) -> (ReturnValue:Fort Item Object Reference)
   > n3 = Is Valid(InputObject=@n2)
   Return Node!(execute=@n3.Is Not Valid, RemainingTime:double=0, TotalDuration:double=0)
   > n4 = GetCoolDownTimeRemaining!(
    execute
    self:Fort Item Object Reference=@n2
    FortPC:Fort Player Controller Object Reference=Cast To FortPlayerController!(Object=GetOwningPlayer()) -> (AsFort Player Controller, bSuccess:bool=~)
   ) -> (then, CooldownDuration:float, ReturnValue:float)
   > Return Node(RemainingTime=@n4.ReturnValue, TotalDuration=@n4)
  ]}
 ResetMaterialProperties {nodes [
   ResetMaterialProperties() {is=title K2Node_FunctionEntry}
   > SetScalarParameterValue(self=Get ReticleDynamicMaterial(), ParameterName=ActiveDash, Value=0)
   > SetScalarParameterValue(self=Get ReticleDynamicMaterial(), ParameterName=InnerCircleRingGapRotation, Value=0)
   > SetScalarParameterValue(self=Get ReticleDynamicMaterial(), ParameterName=ActiveJump, Value=0)
   > SetScalarParameterValue(self=Get ReticleDynamicMaterial(), ParameterName=Slam/Punch, Value=0)
   > SetScalarParameterValue(self=Get ReticleDynamicMaterial(), ParameterName=Distortion, Value=0)
   > SetScalarParameterValue(self=Get ReticleDynamicMaterial(), ParameterName=Dash Directional Opacity, Value=0)
  ]}
]
variables [
 Audio_AmmoCount {type=integer, default=4096}
 DashAbilityStartedTag {type=GameplayTag, default {tagName=RiftGauntlets.Ability.Dash.Intro}}
 DashAbilityTriggeredAsyncAction {type=AbilityAsync_WaitGameplayTagAdded, default=null}
 DashArrowEventTimerHandle {type=TimerHandle}
 DashMovementEndedAsyncAction {type=AbilityAsync_WaitGameplayEvent, default=null}
 DashMovementEndedGameplayTag {type=GameplayTag, default {tagName=RiftGauntlets.Event.Ability.Dash.MME.Stop}}
 DashMovementStartAsyncAction {type=AbilityAsync_WaitGameplayEvent, default=null}
 DashMovementStartedGameplayTag {type=GameplayTag, default {tagName=RiftGauntlets.Event.Ability.Dash.MME.Start}}
 Hotfix_EnableDashDirectionArrowHUD {
  type=ScalableFloat
  default {
   registryType=None
   value=1
   curve {curveTable=/RiftGauntletsGameplay/DataTables/RiftGauntletsGameplayGameData, rowName=RiftGauntlets.HUD.DashDirectionArrow.Enable}
  }
 }
 JumpAbilityStartedTag {type=GameplayTag, default {tagName=RiftGauntlets.Event.Ability.Ascend.MME.Start}}
 JumpAsyncAction {type=AbilityAsync_WaitGameplayEvent, default=null}
 PerformedDamageAsyncAction {type=AbilityAsync_WaitGameplayTagAdded, default=null}
 PerformedDamageGameplayTag {type=GameplayTag, default {tagName=RiftGauntlets.PerformedDamage}}
 Pip Percent Mat Param Name {type=string, default=CooldownProgress}
 ReticleDynamicMaterial {type=MaterialInstanceDynamic}
 RiftGauntletTargetingFeedbackComponent {type=B_RiftGauntletsAbilityTargetingFeedbackComponent_C, default=null}
 RiftGauntletWeaponComponent {type=B_Weapon_RiftGauntlets_C, default=null}
 ViewTargetPlayerPawn {type=FortPlayerPawn}
]
widgets {
 info {
  inheritedWidgetCount=1
  parentClass=/Game/Athena/HUD/reticles/ReticleExtensionWidgets/ChargeStack/WBP_Reticle_RechargeStackBase.WBP_Reticle_RechargeStackBase_C
  rootWidgetClass=&Overlay
  widgetCount=5
 }
 widgets [
  Overlay_22 {widgetClassPath=&Overlay, widget=Overlay_22}
  Image_Reticle {
   widgetClassPath=&Image
   bIsVariable=true
   parent=Overlay_22
   widget=Image_Reticle
   properties {
    visibility=SelfHitTestInvisible
    brush {
     bIsDynamicallyLoaded=false
     drawAs=Image
     imageType=FullColor
     mirroring=NoMirror
     resourceName=None
     resourceObject=/RiftGauntletsGameplay/UI/Art/Materials/M_UI_RiftGauntlets_Reticle_New
     tiling=NoTile
     imageSize=<120 120>
     margin=<ltrb 0 0 0 0>
     outlineSettings {
      bUseBrushTransparency=false
      roundingType=HalfHeightRadius
      width=0
      color {colorUseRule=UseColor_Specified, specifiedColor=<rgba 0 0 0 0>}
      cornerRadii=<0 0 0 0>
     }
     tintColor {colorUseRule=UseColor_Specified, specifiedColor=<rgba 1 1 1 1>}
     uVRegion {bIsValid=false, max=<0 0>, min=<0 0>}
    }
   }
   slot {class=OverlaySlot, properties {horizontalAlignment=HAlign_Center, verticalAlignment=VAlign_Center}}
  }
  VerticalBox_0 {
   widgetClassPath=&VerticalBox
   parent=Overlay_22
   widget=VerticalBox_0
   slot {class=OverlaySlot, properties {horizontalAlignment=HAlign_Center, verticalAlignment=VAlign_Center, padding=<ltrb 0 60 0 0>}}
  }
  WBP_UI_RiftGauntlets_MaterialOnly {
   widgetClassPath=&/WBP_UI_RiftGauntlets_MaterialOnly
   bIsVariable=true
   parent=VerticalBox_0
   widget=WBP_UI_RiftGauntlets_MaterialOnly
   slot {class=VerticalBoxSlot}
  }
  TutorialHighlight {widgetClassPath=&TutorialHighlight}
 ]
}
