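# Rift Gauntlets ammo pip widget that draws up to five recharging pips in a single combined material between two decorative end caps.
# SetMaxAmmo async-loads the material matching the clamped pip count and repositions the caps.
# Ammo and cooldown events write per-pip scalar parameters, using names built from pip indices.
# Use and fill-ready animations key widget variables, which UpdateMaterialPropreties pushes to the animated pip's parameters.
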WidgetBlueprint /RiftGauntletsGameplay/UI/Widgets/WBP_UI_RiftGauntlets_MaterialOnly
animations [
 AnimFillReady {
  binding_count=1
  length_seconds=0.4
  object=&self:AnimFillReady
  sequence {
   bindings [WBP_UI_RiftGauntlets_MaterialOnly_C_112 {
     id=#1
     possessed_class=WBP_UI_RiftGauntlets_MaterialOnly_C
     tracks [
      {
       type=MovieSceneDoubleTrack
       display_name=Gradient
       sections [{
         type=MovieSceneDoubleSection
         end_frame=8
         start_frame=0
         channels [None {type=MovieSceneScriptingDoubleChannel, default=0, keys [[0,0],[3,1],[8,0]]}]
        }]
      }
      {
       type=MovieSceneDoubleTrack
       display_name=Size
       sections [{
         type=MovieSceneDoubleSection
         end_frame=8
         start_frame=0
         channels [None {type=MovieSceneScriptingDoubleChannel, keys [[0,0.35],[4,0.32],[8,0.35]]}]
        }]
      }
      {
       type=MovieSceneDoubleTrack
       display_name=Fill Opacity
       sections [{type=MovieSceneDoubleSection, end_frame=0, start_frame=0, channels [None {type=MovieSceneScriptingDoubleChannel, const [0,1]}]}]
      }
      {
       type=MovieSceneDoubleTrack
       display_name=Progress
       sections [{type=MovieSceneDoubleSection, end_frame=0, start_frame=0, channels [None {type=MovieSceneScriptingDoubleChannel, const [1,1]}]}]
      }
     ]
    }]
   tracks [{type=MovieSceneEventTrack, sections [{type=MovieSceneEventRepeaterSection, end_frame=8, start_frame=0}]}]
  }
 }
 AnimUse {
  binding_count=1
  length_seconds=0.4
  object=&self:AnimUse
  sequence {
   bindings [WBP_UI_RiftGauntlets_MaterialOnly_C_112 {
     id=#1
     possessed_class=WBP_UI_RiftGauntlets_MaterialOnly_C
     tracks [
      {
       type=MovieSceneDoubleTrack
       display_name=Gradient
       sections [{type=MovieSceneDoubleSection, end_frame=8, start_frame=0, channels [None {type=MovieSceneScriptingDoubleChannel, default=0, const [0,2]}]}]
      }
      {
       type=MovieSceneDoubleTrack
       display_name=Size
       sections [{
         type=MovieSceneDoubleSection
         end_frame=8
         start_frame=0
         channels [None {type=MovieSceneScriptingDoubleChannel, keys [[0,0.35],[4,0.4],[8,0.35]]}]
        }]
      }
      {
       type=MovieSceneDoubleTrack
       display_name=Fill Opacity
       sections [{type=MovieSceneDoubleSection, end_frame=0, start_frame=0, channels [None {type=MovieSceneScriptingDoubleChannel, const [0,1]}]}]
      }
      {
       type=MovieSceneDoubleTrack
       display_name=Progress
       sections [{type=MovieSceneDoubleSection, end_frame=8, start_frame=0, channels [None {type=MovieSceneScriptingDoubleChannel, keys [[0,1],[8,0]]}]}]
      }
     ]
    }]
   tracks [{type=MovieSceneEventTrack, sections [{type=MovieSceneEventRepeaterSection, end_frame=8, start_frame=0}]}]
  }
 }
]
graphs [
 ConstructPipParameterName {nodes [
   n1 = ConstructPipParameterName!() -> (then, InInt:int, ParameterName:String) {is=title K2Node_FunctionEntry}
   n2 = "To String (Integer)"(InInt=@n1)
   > Return Node!(execute, ReturnValue:Name=String To Name(InString=Append!(A=Pip, B=@n2, C:String=@n1.ParameterName) -> (ReturnValue)))
  ]}
 EventGraph {nodes [
   n1 = OnAmmoChanged!() -> (OutputDelegate, then, MaxAmmo:int, CurrentAmmo:int) {is=title K2Node_CustomEvent}
   n2 = "Clamp (Integer)"(Value=@n1.CurrentAmmo, Min=0, Max=5) -> (ReturnValue=0)
   > SetNumberOfFilledProgress!(execute, self, CurrentIndex:int=@n2) -> (then)
   > n3 = Branch(Condition=Get CanPlayAnimations() -> (CanPlayAnimations:bool))
   > n4 = Set ShouldShift(ShouldShift:bool="integer < integer"(A=@n2, B=Get CurrentCompletedAmmo() -> (CurrentCompletedAmmo:int=0))) -> (Output_Get:bool)
   > n5 = Branch(Condition=@n4)
   > n6 = PlayAnimationForward(InAnimation=Get AnimUse() -> (AnimUse:Widget Animation Object Reference))
   Set CurrentCompletedAmmo(execute=[@n6.then, @n7.then, @n3.else], CurrentCompletedAmmo:int=@n2) -> (Output_Get:int=0)
   n7 = PlayAnimationForward(execute=@n5.else, InAnimation=Get AnimFillReady() -> (AnimFillReady:Widget Animation Object Reference))
   n8 = CooldownUpdated!() -> (OutputDelegate, then, CurrentTime:double=~, TotalTime:double=~) {is=title K2Node_CustomEvent}
   > SetProgressForCurrentPip!(execute, self, Progress:double=MapRangeClamped(Value=SafeDivide(A=@n8.CurrentTime, B=@n8.TotalTime), InRangeA=1, InRangeB=0)) -> (then)
   UpdateMaterialPropreties() {is=title K2Node_CustomEvent}
   > SetScalarParameterValue(
    self=GetMaterial!() -> (Output_Get:Material Instance Dynamic Object Reference)
    ParameterName=ConstructPipParameterName!(self, InInt:int=GetPipToAnimate!(self) -> (Out:int=0), ParameterName:String=Gradient) -> (ReturnValue:Name=None) {is=title K2Node_CallFunction}
    Value=Get Gradient() -> (Gradient:double)
   )
   > SetScalarParameterValue(
    self=GetMaterial()
    ParameterName=ConstructPipParameterName(InInt=GetPipToAnimate() -> (Out=0), ParameterName=Size) -> (ReturnValue=None)
    Value=Get Size() -> (Size:double)
   )
   > SetScalarParameterValue(
    self=GetMaterial()
    ParameterName=ConstructPipParameterName(InInt=GetPipToAnimate() -> (Out=0), ParameterName=Opacity) -> (ReturnValue=None)
    Value=Get FillOpacity() -> (FillOpacity:double)
   )
   > SetScalarParameterValue(
    self=GetMaterial()
    ParameterName=ConstructPipParameterName(InInt=GetPipToAnimate() -> (Out=0), ParameterName=Progress) -> (ReturnValue=None)
    Value=Get Progress() -> (Progress:double)
   )
   n9 = SetMaxAmmo!() -> (OutputDelegate, then, CurrentCompletedAmmo:int) {is=title K2Node_CustomEvent}
   > Set PreviousVis(PreviousVis:ESlateVisibility Enum=GetVisibility!(self:Widget Object Reference) -> (ReturnValue:ESlateVisibility Enum=Visible)) -> (Output_Get:ESlateVisibility Enum=Visible)
   > SetVisibility(InVisibility=Hidden)
   > PrintString!(
    execute
    InString=Append!(A="Setting up max ammo: ", B="To String (Integer)"(InInt=@n9.CurrentCompletedAmmo)) -> (ReturnValue)
    bPrintToScreen
    bPrintToLog=false
    TextColor=(R=1,G=0.3746,B=0.0228,A=1)
    Duration
    Key
   ) -> (then)
   > Set CanPlayAnimations(CanPlayAnimations:bool=false) -> (Output_Get:bool)
   > n10 = Branch(Condition="integer > integer"(A=@n9.CurrentCompletedAmmo, B=5))
   n11 = GetDisplayName(Object=Self-Reference())
   > n12 = PrintString(
    InString="To String (Text)"(InText=Format Text!(
      Format="LOCTEXT(\"On {ClassName}: Warning! Trying to set max amount of ammo greater than 5, this is not supported. Please report.\")"
      ClassName:String=@n11
     ) -> (Result))
    TextColor=(R=1,G=0.773,B=0,A=1)
    Duration=3
   )
   Set MaxAmmo(execute=[@n12.then, @n10.else], MaxAmmo:int="Clamp (Integer)"(Value=@n9.CurrentCompletedAmmo, Min=2, Max=5) -> (ReturnValue=0)) -> (Output_Get:int=0)
   n13 = int - int(A=Get MaxAmmo() -> (MaxAmmo:int=0))
   > Set EndMaterial(EndMaterial:Material Soft Object Reference=Select!(
     Option 0:Material Soft Object Reference=/RiftGauntletsGameplay/UI/Art/Materials/M_UI_RiftGauntlets_CombinedPips1
     Option 1:Material Soft Object Reference=/RiftGauntletsGameplay/UI/Art/Materials/M_UI_RiftGauntlets_CombinedPips_2
     Option 2:Material Soft Object Reference=/RiftGauntletsGameplay/UI/Art/Materials/M_UI_RiftGauntlets_CombinedPips3
     Option 3:Material Soft Object Reference=/RiftGauntletsGameplay/UI/Art/Materials/M_UI_RiftGauntlets_CombinedPips4
     Option 4:Material Soft Object Reference=/RiftGauntletsGameplay/UI/Art/Materials/M_UI_RiftGauntlets_CombinedPips5
     Index:int=@n13
    ) -> (ReturnValue:Material Soft Object Reference)) -> (Output_Get:Material Soft Object Reference)
   > n14 = Async Load Asset(Asset=Get EndMaterial() -> (EndMaterial:Material Soft Object Reference))
   n15 = Cast To Material!(execute=@n14.Completed, Object=@n14) -> (then, CastFailed, AsMaterial:Material Object Reference) {is=title K2Node_DynamicCast}
   > SetVisibility(InVisibility=Get PreviousVis() -> (PreviousVis:ESlateVisibility Enum=Visible))
   n16 = Get Image_101() -> (Image_101:Image Object Reference)
   > SetBrushFromMaterial(self=@n16, Material=@n15)
   > n17 = GetDynamicMaterial(self=@n16)
   > Set CachedMaterial(CachedMaterial:Material Instance Dynamic Object Reference=@n17) -> (Output_Get:Material Instance Dynamic Object Reference)
   > InvalidateLayoutAndVolatility(self:Widget Object Reference)
   n18 = Select(Option 0:double=-45, Option 1:double=-30, Option 2:double=-15, Option 3:double=0, Option 4:double=5, Index=@n13) -> (ReturnValue:double)
   n19 = Get Image_DetailL() -> (Image_DetailL:Image Object Reference)
   > n20 = Set members in Widget Transform!(
    execute:Exec
    StructRef:"Widget Transform Structure (by ref)"=Get RenderTransform!(self:Widget Object Reference=@n19) -> (RenderTransform:Widget Transform Structure)
    Translation:Vector 2D Structure=MakeVector2D(X=float * float(A=@n18, B=-1) {is=title K2Node_PromotableOperator}, Y=0)
   ) -> (then:Exec, StructOut:"Widget Transform Structure (by ref)") {is=title K2Node_SetFieldsInStruct}
   > SetRenderTransform!(execute, self:Widget Object Reference=@n19, InTransform:Widget Transform Structure=@n20) -> (then)
   n21 = Get Image_DetailR() -> (Image_DetailR:Image Object Reference)
   > n22 = Set members in Widget Transform(StructRef=Get RenderTransform(self=@n21), Translation=MakeVector2D(X=@n18, Y=0))
   > SetRenderTransform(self=@n21, InTransform=@n22)
   n23 = Get CurrentCompletedAmmo() -> (CurrentCompletedAmmo=0)
   > OnAmmoChanged!(execute, self, MaxAmmo:int=@n23, CurrentAmmo:int=@n23) -> (then) {is=title K2Node_CallFunction}
   > Set CanPlayAnimations(CanPlayAnimations=true)
   n24 = Event PreConstruct()
   > Branch(Condition=@n24.IsDesignTime)
   > SetMaxAmmo!(execute, self, CurrentCompletedAmmo:int=Get DebugMaxAmmo() -> (DebugMaxAmmo:int=0)) -> (then) {is=title K2Node_CallFunction}
  ]}
 GetCachedWeaponFromParent {nodes [
   GetCachedWeaponFromParent() {is=title K2Node_FunctionEntry}
   > n1 = Get!(execute:Exec) -> (then:Exec, else:Exec, CachedParentReticle:WBP UI Rift Gauntlets Reticle Object Reference)
   Return Node!(
    execute=[@n3.then, @n1.then]
    CachedEquippedWeapon:Fort Weapon Object Reference=Get CachedEquippedWeapon!(self:WBP Reticle Recharge Stack Base Object Reference=Get CachedParentReticle() -> (CachedParentReticle:WBP UI Rift Gauntlets Reticle Object Reference)) -> (CachedEquippedWeapon:Fort Weapon Object Reference)
   )
   n2 = FindParentWidgetOfType!(
    execute=@n1.else
    StartingWidget:Widget Object Reference=Self-Reference()
    Type:Widget Class Reference=&/WBP_UI_Rift_Gauntlets_Reticle.WBP_UI_Rift_Gauntlets_Reticle_C
   ) -> (then, ReturnValue:WBP UI Rift Gauntlets Reticle Object Reference)
   > n3 = Set CachedParentReticle(CachedParentReticle:WBP UI Rift Gauntlets Reticle Object Reference=@n2) -> (Output_Get:WBP UI Rift Gauntlets Reticle Object Reference)
  ]}
 GetMaterial {nodes [
   GetMaterial() {is=title K2Node_FunctionEntry}
   > n1 = Get!(execute) -> (then, else, CachedMaterial:Material Instance Dynamic Object Reference)
   > Return Node!(execute, Output_Get:Material Instance Dynamic Object Reference=@n1)
   n2 = GetDynamicMaterial(execute=@n1.else, self=Get Image_101())
   > n3 = Set CachedMaterial(CachedMaterial=@n2)
   > Return Node(Output_Get=@n3)
  ]}
 GetPipToAnimate {nodes [
   GetPipToAnimate() {is=title K2Node_FunctionEntry}
   n1 = Get ShouldShift() -> (ShouldShift:bool)
   > Return Node!(execute, Out:int=int + int(A=Get CurrentCompletedAmmo() -> (CurrentCompletedAmmo=0), B=SelectInt(A=1, B=0, bPickA=@n1)))
  ]}
 Sequencer Events {links [], nodes []}
 SetNumberOfFilledProgress {nodes [
   n1 = SetNumberOfFilledProgress!() -> (then, CurrentIndex:int) {is=title K2Node_FunctionEntry}
   > SetScalarParameterValue(self=GetMaterial() {is=title K2Node_CallFunction}, ParameterName=CurrentAmmo, Value="To Float (Integer)"(InInt=int + int(A=@n1)))
   > n2 = For Loop(FirstIndex=1, LastIndex=4)
   n3 = For Loop(execute=@n2.Completed, FirstIndex=1, LastIndex=@n1)
   n4 = GetMaterial()
   > SetScalarParameterValue(self=@n4, ParameterName=ConstructPipParameterName(InInt=@n3, ParameterName=Progress) -> (ReturnValue=None), Value=1)
   > SetScalarParameterValue(self=@n4, ParameterName=ConstructPipParameterName(InInt=@n3, ParameterName=Opacity) -> (ReturnValue=None), Value=0)
   SetScalarParameterValue(
    execute=@n2.LoopBody
    self=GetMaterial()
    ParameterName=ConstructPipParameterName(InInt=@n2, ParameterName=Progress) -> (ReturnValue=None)
    Value=0
   )
  ]}
 SetProgressForCurrentPip {nodes [
   n1 = SetProgressForCurrentPip!() -> (then, Progress:double=~) {is=title K2Node_FunctionEntry}
   > SetScalarParameterValue(self=GetMaterial(), ParameterName=CurrentAmmo, Value=float + float(A=@n1, B:int=Get CurrentCompletedAmmo() -> (CurrentCompletedAmmo=0)))
   n2 = int + int(A=Get CurrentCompletedAmmo() -> (CurrentCompletedAmmo=0))
   n3 = "integer >= integer"(
    A=GetRemainingAmmo!(self:Fort Weapon Object Reference=GetCachedWeaponFromParent!() -> (CachedEquippedWeapon:Fort Weapon Object Reference) {is=title K2Node_CallFunction}) -> (ReturnValue:int=0)
    B=@n2
   )
   n4 = GetMaterial()
   > SetScalarParameterValue(
    self=@n4
    ParameterName=ConstructPipParameterName(InInt=@n2, ParameterName=Progress) -> (ReturnValue=None)
    Value=SelectFloat(A=Get Progress(), B=0, bPickA=@n3)
   )
   > SetScalarParameterValue(
    self=@n4
    ParameterName=ConstructPipParameterName(InInt=@n2, ParameterName=Opacity) -> (ReturnValue=None)
    Value=SelectFloat(A=1, B=0, bPickA=@n3)
   )
  ]}
]
variables [
 CachedMaterial {type=MaterialInstanceDynamic}
 CachedParentReticle {type=WBP_UI_Rift_Gauntlets_Reticle_C, default=null}
 CanPlayAnimations {type=boolean}
 CurrentCompletedAmmo {type=integer, default=-10}
 DebugMaxAmmo {type=integer, default=3}
 EndMaterial {type=Material, default=null, class=~, graph=~}
 FillOpacity {type=number}
 Gradient {type=number}
 MaxAmmo {type=integer}
 PreviousVis {type=ESlateVisibility, default=Visible}
 Progress {type=number}
 ShouldShift {type=boolean}
 Size {type=number}
]
widgets {
 info {inheritedWidgetCount=0, parentClass=&UserWidget, rootWidgetClass=&SizeBox, widgetCount=5}
 widgets [
  SizeBox_27 {
   widgetClassPath=&SizeBox
   widget=SizeBox_27
   properties {bOverride_HeightOverride=true, bOverride_WidthOverride=true, heightOverride=28, widthOverride=118}
  }
  HorizontalBox_0 {
   widgetClassPath=&HorizontalBox
   parent=SizeBox_27
   widget=HorizontalBox_0
   slot {class=SizeBoxSlot, properties {verticalAlignment=VAlign_Center}}
  }
  Image_DetailL {
   widgetClassPath=&Image
   bIsVariable=true
   parent=HorizontalBox_0
   widget=Image_DetailL
   properties {
    visibility=SelfHitTestInvisible
    brush {
     bIsDynamicallyLoaded=false
     drawAs=Image
     imageType=FullColor
     mirroring=NoMirror
     resourceName=None
     resourceObject=/RiftGauntletsGameplay/UI/Art/Materials/M_UI_RiftGauntlets_Pip_Decor
     tiling=NoTile
     imageSize=<16 16>
     margin=<ltrb 0 0 0 0>
     outlineSettings {
      bUseBrushTransparency=false
      roundingType=HalfHeightRadius
      width=0
      color {colorUseRule=UseColor_Specified, specifiedColor=<rgba 0 0 0 0>}
      cornerRadii=<0 0 0 0>
     }
     tintColor {colorUseRule=UseColor_Specified, specifiedColor=<rgba 1 1 1 1>}
     uVRegion {bIsValid=false, max=<0 0>, min=<0 0>}
    }
    renderTransform {angle=0, scale=<-1 1>, shear=<0 0>, translation=<0 0>}
   }
   slot {class=HorizontalBoxSlot, properties {horizontalAlignment=HAlign_Center, verticalAlignment=VAlign_Center, padding=<ltrb 0 -2 -1 0>}}
  }
  Image_101 {
   widgetClassPath=&Image
   bIsVariable=true
   parent=HorizontalBox_0
   widget=Image_101
   properties {
    brush {
     bIsDynamicallyLoaded=false
     drawAs=Image
     imageType=FullColor
     mirroring=NoMirror
     resourceName=None
     resourceObject=&M_UI_RiftGauntlets_CombinedPips5
     tiling=NoTile
     imageSize=<34.5213 32>
     margin=<ltrb 0 0 0 0>
     outlineSettings {
      bUseBrushTransparency=false
      roundingType=HalfHeightRadius
      width=0
      color {colorUseRule=UseColor_Specified, specifiedColor=<rgba 0 0 0 0>}
      cornerRadii=<0 0 0 0>
     }
     tintColor {colorUseRule=UseColor_Specified, specifiedColor=<rgba 1 1 1 1>}
     uVRegion {bIsValid=false, max=<0 0>, min=<0 0>}
    }
    renderTransform {angle=0, scale=<1 1>, shear=<0 0>, translation=<0 -1>}
   }
   slot {class=HorizontalBoxSlot, properties {size {sizeRule=Fill, value=1}}}
  }
  Image_DetailR {
   widgetClassPath=&Image
   bIsVariable=true
   parent=HorizontalBox_0
   widget=Image_DetailR
   properties {
    visibility=SelfHitTestInvisible
    brush {
     bIsDynamicallyLoaded=false
     drawAs=Image
     imageType=FullColor
     mirroring=NoMirror
     resourceName=None
     resourceObject=&M_UI_RiftGauntlets_Pip_Decor
     tiling=NoTile
     imageSize=<16 16>
     margin=<ltrb 0 0 0 0>
     outlineSettings {
      bUseBrushTransparency=false
      roundingType=HalfHeightRadius
      width=0
      color {colorUseRule=UseColor_Specified, specifiedColor=<rgba 0 0 0 0>}
      cornerRadii=<0 0 0 0>
     }
     tintColor {colorUseRule=UseColor_Specified, specifiedColor=<rgba 1 1 1 1>}
     uVRegion {bIsValid=false, max=<0 0>, min=<0 0>}
    }
   }
   slot {class=HorizontalBoxSlot, properties {horizontalAlignment=HAlign_Center, verticalAlignment=VAlign_Center, padding=<ltrb -1 -2 0 0>}}
  }
 ]
}
